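# Post-process-style AnimBP for the NanoChain female body cosmetic, adding secondary motion to pants, belts, collar, zip and cans.
# The parent mesh's copied pose runs through a state-driven Control Rig, chained AnimDynamics and a jump belt tilt into a cached pose.
# Two pant simulations branch from it, with the angular-spring version chosen in the front end.
# Event logic sets alphas from movement, emote and zipline state, holding three right-belt simulations at zero.
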AnimBlueprint /BRCosmetics/Characters/Player/Female/Medium/Bodies/F_MED_NanoChain/Meshes/F_MED_NanoChain_AnimBP
target_skeleton=&/F_MED_NanoChain_Skeleton
graphs [
 AnimGraph {nodes [
   Output Pose(Result=Blend Poses by bool(BlendPose_0=Use cached pose 'FEPAnts'(), BlendPose_1=Use cached pose 'Pants'(), bActiveValue=Get bIsInFrontEnd()) {blendType=Linear}) {layerGroup=None}
   n1 = Get NoEmote() -> (NoEmote:double)
   n2 = Get GravityOverride()
   n3 = Anim Dynamics - Bone: dyn_pantf_r!(ComponentPose=Local To Component(LocalPose=Use cached pose 'MainLopp'()), Alpha=@n1, GravityOverride:Vector=@n2, ExternalForce="0,0,0") -> (Pose) {
    angularDampingOverride=1
    angularSpringConstant=1000
    bOverrideAngularDamping=true
    bUseGravityOverride=true
    gravityScale=0.35
    simulationSpace=BoneRelative
    alphaBoolBlend {blendOption=Linear}
    boundBone {boneName=dyn_pantf_r}
    componentAppliedLinearAccClamp=<3000 3000 3000>
    componentLinearAccScale=<0.5 0.5 0.5>
    componentLinearVelScale=<0.5 0.5 0.5>
    gravityOverride {z=-980}
    physicsBodyDefinitions [{
      boundBone {boneName=dyn_pantf_r}
      boxExtents=<5 5 15>
      constraintSetup {angularLimitsMax=<0 0 0>, angularLimitsMin=<-60 -10 -10>, angularTarget=<1 0 0>}
      localJointOffset=<0 0 10>
     }]
    relativeSpaceBone {boneName=base_pant_r}
   }
   n4 = Anim Dynamics - Bone: dyn_pantf_02_r!(ComponentPose=@n3, Alpha=@n1, GravityOverride:Vector=@n2, ExternalForce="0,0,0") -> (Pose) {
    angularDampingOverride=1
    angularSpringConstant=2000
    bOverrideAngularDamping=true
    bUseGravityOverride=true
    gravityScale=0.35
    simulationSpace=RootRelative
    alphaBoolBlend {blendOption=Linear}
    boundBone {boneName=dyn_pantf_02_r}
    componentAppliedLinearAccClamp=<3000 3000 3000>
    componentLinearAccScale=<0.3 0.3 0.3>
    componentLinearVelScale=<0.3 0.3 0.3>
    gravityOverride {z=-980}
    physicsBodyDefinitions [{
      boundBone {boneName=dyn_pantf_02_r}
      boxExtents=<10 10 20>
      constraintSetup {angularLimitsMax=<0 0 0>, angularLimitsMin=<-20 -30 -5>, angularTarget=<1 0 0>}
      localJointOffset=<0 0 10>
     }]
   }
   n5 = Anim Dynamics - Bone: dyn_pantb_02_r!(ComponentPose=@n4, Alpha=@n1, GravityOverride:Vector=@n2, ExternalForce="0,0,0") -> (Pose) {
    angularDampingOverride=1
    angularSpringConstant=100
    bOverrideAngularDamping=true
    bUseGravityOverride=true
    gravityScale=0.35
    simulationSpace=RootRelative
    alphaBoolBlend {blendOption=Linear}
    boundBone {boneName=dyn_pantb_02_r}
    componentAppliedLinearAccClamp=<3000 3000 3000>
    componentLinearAccScale=<0.3 0.3 0.3>
    componentLinearVelScale=<0.3 0.3 0.3>
    gravityOverride {z=-980}
    physicsBodyDefinitions [{
      boundBone {boneName=dyn_pantb_02_r}
      boxExtents=<10 10 20>
      constraintSetup {angularLimitsMax=<0 0 0>, angularLimitsMin=<-40 0 0>, angularTarget=<1 0 0>}
      localJointOffset=<0 0 10>
     }]
   }
   n6 = Anim Dynamics - Bone: dyn_pantf_l!(ComponentPose=@n5, Alpha=@n1, GravityOverride:Vector=@n2, ExternalForce="0,0,0") -> (Pose) {
    angularDampingOverride=1
    angularSpringConstant=1000
    bOverrideAngularDamping=true
    bUseGravityOverride=true
    gravityScale=0.35
    simulationSpace=BoneRelative
    alphaBoolBlend {blendOption=Linear}
    boundBone {boneName=dyn_pantf_l}
    componentAppliedLinearAccClamp=<3000 3000 3000>
    componentLinearAccScale=<0.5 0.5 0.5>
    componentLinearVelScale=<0.5 0.5 0.5>
    gravityOverride {z=-980}
    physicsBodyDefinitions [{
      boundBone {boneName=dyn_pantf_l}
      boxExtents=<5 5 15>
      constraintSetup {angularLimitsMax=<0 10 0>, angularLimitsMin=<-60 0 -10>, angularTarget=<1 0 0>}
      localJointOffset=<0 0 10>
     }]
    relativeSpaceBone {boneName=base_pant_l}
   }
   n7 = Anim Dynamics - Bone: dyn_pantf_02_l!(ComponentPose=@n6, Alpha=@n1, GravityOverride:Vector=@n2, ExternalForce="0,0,0") -> (Pose) {
    angularDampingOverride=1
    angularSpringConstant=200
    bOverrideAngularDamping=true
    bUseGravityOverride=true
    gravityScale=0.35
    simulationSpace=RootRelative
    alphaBoolBlend {blendOption=Linear}
    boundBone {boneName=dyn_pantf_02_l}
    componentAppliedLinearAccClamp=<3000 3000 3000>
    componentLinearAccScale=<0.3 0.3 0.3>
    componentLinearVelScale=<0.3 0.3 0.3>
    gravityOverride {z=-980}
    physicsBodyDefinitions [{
      boundBone {boneName=dyn_pantf_02_l}
      boxExtents=<10 10 20>
      constraintSetup {angularLimitsMax=<0 0 0>, angularLimitsMin=<-20 -30 -5>, angularTarget=<1 0 0>}
      localJointOffset=<0 0 10>
     }]
   }
   n8 = Anim Dynamics - Bone: dyn_pantb_02_l!(ComponentPose=@n7, Alpha=@n1, GravityOverride:Vector=@n2, ExternalForce="0,0,0") -> (Pose) {
    angularDampingOverride=1
    angularSpringConstant=100
    bOverrideAngularDamping=true
    bUseGravityOverride=true
    gravityScale=0.35
    simulationSpace=RootRelative
    alphaBoolBlend {blendOption=Linear}
    boundBone {boneName=dyn_pantb_02_l}
    componentAppliedLinearAccClamp=<3000 3000 3000>
    componentLinearAccScale=<0.3 0.3 0.3>
    componentLinearVelScale=<0.3 0.3 0.3>
    gravityOverride {z=-980}
    physicsBodyDefinitions [{
      boundBone {boneName=dyn_pantb_02_l}
      boxExtents=<10 10 20>
      constraintSetup {angularLimitsMax=<0 0 0>, angularLimitsMin=<-40 0 0>, angularTarget=<1 0 0>}
      localJointOffset=<0 0 10>
     }]
   }
   Save cached pose 'Pants'(Pose=Component To Local(ComponentPose=@n8)) {cacheName=Pants}
   n9 = Get GravityOverride()
   n10 = Get NoEmote()
   n11 = Anim Dynamics - Bone: dyn_pantf_r(ComponentPose=Local To Component(LocalPose=Use cached pose 'MainLopp'()), Alpha=@n10, GravityOverride=@n9, ExternalForce="0,0,0") {
    angularDampingOverride=1
    angularSpringConstant=2000
    bAngularSpring=true
    bOverrideAngularDamping=true
    bUseGravityOverride=true
    gravityScale=0.35
    simulationSpace=BoneRelative
    alphaBoolBlend {blendOption=Linear}
    boundBone {boneName=dyn_pantf_r}
    componentAppliedLinearAccClamp=<3000 3000 3000>
    componentLinearAccScale=<0.5 0.5 0.5>
    componentLinearVelScale=<0.5 0.5 0.5>
    gravityOverride {z=-980}
    physicsBodyDefinitions [{
      boundBone {boneName=dyn_pantf_r}
      boxExtents=<5 5 15>
      constraintSetup {angularLimitsMax=<0 0 0>, angularLimitsMin=<-60 -10 -10>, angularTarget=<1 0 0>}
      localJointOffset=<0 0 8>
     }]
    relativeSpaceBone {boneName=base_pant_r}
   }
   n12 = Anim Dynamics - Bone: dyn_pantf_02_r(ComponentPose=@n11, Alpha=@n10, GravityOverride=@n9, ExternalForce="0,0,0") {
    angularDampingOverride=1
    angularSpringConstant=2000
    bAngularSpring=true
    bOverrideAngularDamping=true
    bUseGravityOverride=true
    gravityScale=0.35
    simulationSpace=RootRelative
    alphaBoolBlend {blendOption=Linear}
    boundBone {boneName=dyn_pantf_02_r}
    componentAppliedLinearAccClamp=<3000 3000 3000>
    componentLinearAccScale=<0.3 0.3 0.3>
    componentLinearVelScale=<0.3 0.3 0.3>
    gravityOverride {z=-980}
    physicsBodyDefinitions [{
      boundBone {boneName=dyn_pantf_02_r}
      boxExtents=<10 10 20>
      constraintSetup {angularLimitsMax=<0 0 0>, angularLimitsMin=<-20 -30 -5>, angularTarget=<1 0 0>}
      localJointOffset=<0 0 8>
     }]
   }
   n13 = Anim Dynamics - Bone: dyn_pantb_02_r(ComponentPose=@n12, Alpha=@n10, GravityOverride=@n9, ExternalForce="0,0,0") {
    angularDampingOverride=1
    angularSpringConstant=2000
    bAngularSpring=true
    bOverrideAngularDamping=true
    bUseGravityOverride=true
    gravityScale=0.35
    simulationSpace=RootRelative
    alphaBoolBlend {blendOption=Linear}
    boundBone {boneName=dyn_pantb_02_r}
    componentAppliedLinearAccClamp=<3000 3000 3000>
    componentLinearAccScale=<0.3 0.3 0.3>
    componentLinearVelScale=<0.3 0.3 0.3>
    gravityOverride {z=-980}
    physicsBodyDefinitions [{
      boundBone {boneName=dyn_pantb_02_r}
      boxExtents=<10 10 20>
      constraintSetup {angularLimitsMax=<0 0 0>, angularLimitsMin=<-40 0 0>, angularTarget=<1 0 0>}
      localJointOffset=<0 0 8>
     }]
   }
   n14 = Anim Dynamics - Bone: dyn_pantf_l(ComponentPose=@n13, Alpha=@n10, GravityOverride=@n9, ExternalForce="0,0,0") {
    angularDampingOverride=1
    angularSpringConstant=2000
    bAngularSpring=true
    bOverrideAngularDamping=true
    bUseGravityOverride=true
    gravityScale=0.35
    simulationSpace=BoneRelative
    alphaBoolBlend {blendOption=Linear}
    boundBone {boneName=dyn_pantf_l}
    componentAppliedLinearAccClamp=<3000 3000 3000>
    componentLinearAccScale=<0.5 0.5 0.5>
    componentLinearVelScale=<0.5 0.5 0.5>
    gravityOverride {z=-980}
    physicsBodyDefinitions [{
      boundBone {boneName=dyn_pantf_l}
      boxExtents=<5 5 15>
      constraintSetup {angularLimitsMax=<0 10 0>, angularLimitsMin=<-60 0 -10>, angularTarget=<1 0 0>}
      localJointOffset=<0 0 8>
     }]
    relativeSpaceBone {boneName=base_pant_l}
   }
   n15 = Anim Dynamics - Bone: dyn_pantf_02_l(ComponentPose=@n14, Alpha=@n10, GravityOverride=@n9, ExternalForce="0,0,0") {
    angularDampingOverride=1
    angularSpringConstant=2000
    bAngularSpring=true
    bOverrideAngularDamping=true
    bUseGravityOverride=true
    gravityScale=0.35
    simulationSpace=RootRelative
    alphaBoolBlend {blendOption=Linear}
    boundBone {boneName=dyn_pantf_02_l}
    componentAppliedLinearAccClamp=<3000 3000 3000>
    componentLinearAccScale=<0.3 0.3 0.3>
    componentLinearVelScale=<0.3 0.3 0.3>
    gravityOverride {z=-980}
    physicsBodyDefinitions [{
      boundBone {boneName=dyn_pantf_02_l}
      boxExtents=<10 10 20>
      constraintSetup {angularLimitsMax=<0 0 0>, angularLimitsMin=<-20 -30 -5>, angularTarget=<1 0 0>}
      localJointOffset=<0 0 8>
     }]
   }
   n16 = Anim Dynamics - Bone: dyn_pantb_02_l(ComponentPose=@n15, Alpha=@n10, GravityOverride=@n9, ExternalForce="0,0,0") {
    angularDampingOverride=1
    angularSpringConstant=2000
    bAngularSpring=true
    bOverrideAngularDamping=true
    bUseGravityOverride=true
    gravityScale=0.35
    simulationSpace=RootRelative
    alphaBoolBlend {blendOption=Linear}
    boundBone {boneName=dyn_pantb_02_l}
    componentAppliedLinearAccClamp=<3000 3000 3000>
    componentLinearAccScale=<0.3 0.3 0.3>
    componentLinearVelScale=<0.3 0.3 0.3>
    gravityOverride {z=-980}
    physicsBodyDefinitions [{
      boundBone {boneName=dyn_pantb_02_l}
      boxExtents=<10 10 20>
      constraintSetup {angularLimitsMax=<0 0 0>, angularLimitsMin=<-40 0 0>, angularTarget=<1 0 0>}
      localJointOffset=<0 0 8>
     }]
   }
   Save cached pose 'FEPAnts'(Pose=Component To Local(ComponentPose=@n16)) {cacheName=FEPAnts}
   n17 = Control Rig!(
    Source=Copy Pose From Mesh() {bUseAttachedParent=true}
    Alpha
    isCrouching:double=Get Crouching() -> (Crouching:double)
    DBNO:double=Get DBNO() -> (DBNO:double)
    is targeting:double=Get isTargeting() -> (isTargeting:double)
    isFloating:bool=Get skydivefloating() -> (skydivefloating:bool)
    In FE:double=Get FE_NoEmote() -> (FE_NoEmote:double)
    NotCrouching:double=Get NotCrouching() -> (NotCrouching:double)
    Crouch walk:double=Get CrouchMove() -> (CrouchMove:double)
    playing emote:double=Get Emote() -> (Emote:double)
    Swing:double=Get Swing() -> (Swing:double)
    isFrontEnd:bool=Get bIsInFrontEnd()
   ) -> (Pose) {
    alphaBoolBlend {blendOption=Linear}
    alphaScaleBiasClamp {bInterpResult=true}
    controlRigAssetReference {blueprintRigClass=&/F_MED_NanoChain_CtrlRig.F_MED_NanoChain_CtrlRig_C}
   }
   n18 = Get zip Off() -> (zip Off:double)
   n19 = Anim Dynamics - Bone: base_midbelt_l!(ComponentPose=Local To Component(LocalPose=@n17), Alpha=@n18, GravityOverride:Vector="0,0,-2000", ExternalForce="0,0,0") -> (Pose) {
    angularDampingOverride=1
    angularSpringConstant=3000
    bOverrideAngularDamping=true
    bOverrideLinearDamping=true
    bUseGravityOverride=true
    gravityScale=0.35
    linearDampingOverride=1
    linearSpringConstant=500
    simulationSpace=BoneRelative
    alphaBoolBlend {blendOption=Linear}
    boundBone {boneName=base_midbelt_l}
    componentAppliedLinearAccClamp=<1000 1000 1000>
    componentLinearAccScale=<0.2 0.2 0.2>
    componentLinearVelScale=<0.2 0.2 0.2>
    gravityOverride {z=-980}
    physicsBodyDefinitions [{
      boundBone {boneName=base_midbelt_l}
      boxExtents=<10 20 10>
      constraintSetup {angularTargetAxis=AxisZ, angularLimitsMax=<0 0 0>, angularLimitsMin=<0 0 0>, angularTarget=<1 0 0>}
      localJointOffset=<1 0 0>
     }]
   }
   n20 = Get GravityOverride()
   n21 = Anim Dynamics - Bone: dyn_frontcan!(ComponentPose=@n19, Alpha=@n18, GravityOverride:Vector=@n20, ExternalForce="0,0,0") -> (Pose) {
    angularDampingOverride=1
    angularSpringConstant=2000
    bAngularSpring=true
    bOverrideAngularDamping=true
    bUseGravityOverride=true
    gravityScale=0.35
    simulationSpace=RootRelative
    alphaBoolBlend {blendOption=Linear}
    boundBone {boneName=dyn_frontcan}
    componentAppliedLinearAccClamp=<70000 70000 70000>
    componentLinearAccScale=<0.5 0.5 0.5>
    componentLinearVelScale=<0.5 0.5 0.5>
    gravityOverride {z=-980}
    physicsBodyDefinitions [{
      boundBone {boneName=dyn_frontcan}
      boxExtents=<10 5 10>
      constraintSetup {angularLimitsMax=<2 0 2>, angularLimitsMin=<-2 0 -2>, angularTarget=<1 0 0>}
      localJointOffset=<-4 0 0>
     }]
   }
   n22 = Anim Dynamics - Bone: dyn_backcan!(ComponentPose=@n21, Alpha=@n18, GravityOverride:Vector=@n20, ExternalForce="0,0,0") -> (Pose) {
    angularDampingOverride=1
    angularSpringConstant=3000
    bOverrideAngularDamping=true
    bUseGravityOverride=true
    gravityScale=0.35
    simulationSpace=RootRelative
    alphaBoolBlend {blendOption=Linear}
    boundBone {boneName=dyn_backcan}
    componentAppliedLinearAccClamp=<70000 70000 70000>
    componentLinearAccScale=<0.5 0.5 0.5>
    componentLinearVelScale=<0.5 0.5 0.5>
    gravityOverride {z=-980}
    physicsBodyDefinitions [{
      boundBone {boneName=dyn_backcan}
      boxExtents=<10 5 10>
      constraintSetup {angularLimitsMax=<1 0 1>, angularLimitsMin=<-1 0 -1>, angularTarget=<1 0 0>}
      localJointOffset=<-4 0 0>
     }]
   }
   n23 = Anim Dynamics - Bone: dyn_leftcolar(ComponentPose=@n22, Alpha=1, ExternalForce="0,0,0") {
    angularDampingOverride=0.999
    angularSpringConstant=3000
    bAngularSpring=true
    bOverrideAngularDamping=true
    bUseGravityOverride=true
    gravityScale=0.35
    simulationSpace=RootRelative
    alphaBoolBlend {blendOption=Linear}
    boundBone {boneName=dyn_leftcolar}
    componentAppliedLinearAccClamp=<1000 1000 1000>
    componentLinearAccScale=<1 1 1>
    componentLinearVelScale=<1 1 1>
    gravityOverride {z=-980}
    physicsBodyDefinitions [{
      boundBone {boneName=dyn_leftcolar}
      boxExtents=<10 10 10>
      constraintSetup {angularTargetAxis=AxisZ, twistAxis=AxisZ, angularLimitsMax=<10 0 10>, angularLimitsMin=<-10 0 -10>, angularTarget=<1 0 0>}
      localJointOffset=<0 0 4>
     }]
   }
   n24 = Anim Dynamics - Bone: dyn_rightcolar(ComponentPose=@n23, Alpha=1, ExternalForce="0,0,0") {
    angularDampingOverride=0.999
    angularSpringConstant=3000
    bAngularSpring=true
    bOverrideAngularDamping=true
    bUseGravityOverride=true
    gravityScale=0.35
    simulationSpace=RootRelative
    alphaBoolBlend {blendOption=Linear}
    boundBone {boneName=dyn_rightcolar}
    componentAppliedLinearAccClamp=<1000 1000 1000>
    componentLinearAccScale=<1 1 1>
    componentLinearVelScale=<1 1 1>
    gravityOverride {z=-980}
    physicsBodyDefinitions [{
      boundBone {boneName=dyn_rightcolar}
      boxExtents=<10 10 10>
      constraintSetup {angularTargetAxis=AxisZ, twistAxis=AxisZ, angularLimitsMax=<10 0 10>, angularLimitsMin=<-10 0 -10>, angularTarget=<1 0 0>}
      localJointOffset=<0 0 4>
     }]
   }
   n25 = Anim Dynamics - Bone: dyn_backcolar(ComponentPose=@n24, Alpha=1, ExternalForce="0,0,0") {
    angularDampingOverride=0.999
    angularSpringConstant=3000
    bAngularSpring=true
    bOverrideAngularDamping=true
    bUseGravityOverride=true
    gravityScale=0.35
    simulationSpace=RootRelative
    alphaBoolBlend {blendOption=Linear}
    boundBone {boneName=dyn_backcolar}
    componentAppliedLinearAccClamp=<1000 1000 1000>
    componentLinearAccScale=<1 1 1>
    componentLinearVelScale=<1 1 1>
    gravityOverride {z=-980}
    physicsBodyDefinitions [{
      boundBone {boneName=dyn_backcolar}
      boxExtents=<10 10 10>
      constraintSetup {angularTargetAxis=AxisZ, twistAxis=AxisZ, angularLimitsMax=<10 0 10>, angularLimitsMin=<-10 0 -10>, angularTarget=<1 0 0>}
      localJointOffset=<0 0 4>
     }]
   }
   n26 = Anim Dynamics - Bone: dyn_frontcolar_mover(ComponentPose=@n25, Alpha=1, ExternalForce="0,0,0") {
    angularDampingOverride=0.999
    angularSpringConstant=3000
    bAngularSpring=true
    bOverrideAngularDamping=true
    bUseGravityOverride=true
    gravityScale=0.35
    simulationSpace=RootRelative
    alphaBoolBlend {blendOption=Linear}
    boundBone {boneName=dyn_frontcolar_mover}
    componentAppliedLinearAccClamp=<1000 1000 1000>
    componentLinearAccScale=<1 1 1>
    componentLinearVelScale=<1 1 1>
    gravityOverride {z=-980}
    physicsBodyDefinitions [{
      boundBone {boneName=dyn_frontcolar_mover}
      boxExtents=<10 10 10>
      constraintSetup {angularTargetAxis=AxisZ, twistAxis=AxisZ, angularLimitsMax=<0 0 10>, angularLimitsMin=<-10 0 -10>, angularTarget=<1 0 0>}
      localJointOffset=<0 0 4>
     }]
   }
   n27 = Anim Dynamics - Bone: dyn_zip(ComponentPose=@n26, Alpha=1, ExternalForce="0,0,0") {
    angularDampingOverride=1
    angularSpringConstant=500
    bAngularSpring=true
    bOverrideAngularDamping=true
    bUseGravityOverride=true
    gravityScale=0.35
    simulationSpace=BoneRelative
    alphaBoolBlend {blendOption=Linear}
    boundBone {boneName=dyn_zip}
    componentAppliedLinearAccClamp=<1000 1000 1000>
    componentLinearAccScale=<1 1 1>
    componentLinearVelScale=<1 1 1>
    gravityOverride {z=-980}
    physicsBodyDefinitions [{
      boundBone {boneName=dyn_zip}
      boxExtents=<10 10 10>
      constraintSetup {angularTargetAxis=AxisZ, twistAxis=AxisZ, angularLimitsMax=<15 15 15>, angularLimitsMin=<-15 -15 -15>, angularTarget=<1 0 0>}
      localJointOffset=<0 0 5>
     }]
   }
   n28 = Anim Dynamics - Bone: dyn_belt_r(ComponentPose=@n27, Alpha=Get Emote(), ExternalForce="0,0,0") {
    angularDampingOverride=1
    angularSpringConstant=300
    bAngularSpring=true
    bOverrideAngularDamping=true
    bUseGravityOverride=true
    gravityScale=0.35
    linearSpringConstant=100
    simulationSpace=BoneRelative
    alphaBoolBlend {blendOption=Linear}
    boundBone {boneName=dyn_belt_r}
    componentLinearAccScale=<0.5 0.5 0.5>
    componentLinearVelScale=<0.5 0.5 0.5>
    gravityOverride {z=-200}
    physicsBodyDefinitions [{
      boundBone {boneName=dyn_belt_r}
      boxExtents=<10 10 15>
      constraintSetup {
       angularTargetAxis=AxisZ
       angularLimitsMax=<0 0 0>
       angularLimitsMin=<0 -20 0>
       angularTarget=<1 0 0>
       linearAxesMax=<0 0 2>
       linearAxesMin=<0 0 -5>
      }
      localJointOffset=<-5 0 2>
     }]
    relativeSpaceBone {boneName=base_belt_r}
   }
   n29 = Get DynOff() -> (DynOff:double)
   n30 = Anim Dynamics - Bone: dyn_frontbelt_r(ComponentPose=@n28, Alpha=@n29, ExternalForce="0,0,0") {
    angularDampingOverride=1
    angularSpringConstant=1000
    bAngularSpring=true
    bOverrideAngularDamping=true
    bUseGravityOverride=true
    gravityScale=0.35
    simulationSpace=BoneRelative
    alphaBoolBlend {blendOption=Linear}
    boundBone {boneName=dyn_frontbelt_r}
    componentLinearAccScale=<1 1 1>
    componentLinearVelScale=<1 1 1>
    gravityOverride {z=-980}
    physicsBodyDefinitions [{
      boundBone {boneName=dyn_frontbelt_r}
      boxExtents=<10 10 10>
      constraintSetup {angularLimitsMax=<15 15 15>, angularLimitsMin=<-15 -15 -15>, angularTarget=<1 0 0>}
      localJointOffset=<0 0 5>
     }]
    relativeSpaceBone {boneName=base_frontbelt_r}
   }
   n31 = Anim Dynamics - Bone: dyn_midbelt_r(ComponentPose=@n30, Alpha=@n29, ExternalForce="0,0,0") {
    angularDampingOverride=0.9999
    angularSpringConstant=10000
    bAngularSpring=true
    bOverrideAngularDamping=true
    bUseGravityOverride=true
    gravityScale=0.35
    simulationSpace=BoneRelative
    alphaBoolBlend {blendOption=Linear}
    boundBone {boneName=dyn_midbelt_r}
    componentLinearAccScale=<1 1 1>
    componentLinearVelScale=<1 1 1>
    gravityOverride {z=-980}
    physicsBodyDefinitions [{
      boundBone {boneName=dyn_midbelt_r}
      boxExtents=<15 20 20>
      constraintSetup {angularLimitsMax=<15 0 0>, angularLimitsMin=<-35 -40 0>, angularTarget=<1 0 0>}
      localJointOffset=<0 0 8>
     }]
    relativeSpaceBone {boneName=base_midbelt_r}
   }
   n32 = Anim Dynamics - Bone: dyn_backbelt_r(ComponentPose=@n31, Alpha=@n29, ExternalForce="0,0,0") {
    angularDampingOverride=1
    angularSpringConstant=1000
    bAngularSpring=true
    bOverrideAngularDamping=true
    bUseGravityOverride=true
    gravityScale=0.35
    simulationSpace=BoneRelative
    alphaBoolBlend {blendOption=Linear}
    boundBone {boneName=dyn_backbelt_r}
    componentLinearAccScale=<1 1 1>
    componentLinearVelScale=<1 1 1>
    gravityOverride {z=-980}
    physicsBodyDefinitions [{
      boundBone {boneName=dyn_backbelt_r}
      boxExtents=<10 10 10>
      constraintSetup {angularLimitsMax=<0 10 10>, angularLimitsMin=<0 -10 -10>, angularTarget=<1 0 0>}
      localJointOffset=<0 0 5>
     }]
    relativeSpaceBone {boneName=base_backbelt_r}
   }
   n33 = "Transform (Modify) Bone - Bone: base_midbelt_l"(ComponentPose=@n32, Alpha=Get JUMP() -> (JUMP:double), Translation="0,0,0", Rotation="20,0,0", Scale="1,1,1") {
    rotationMode=BMM_Additive
    rotationSpace=BCS_BoneSpace
    translationMode=BMM_Additive
    translationSpace=BCS_BoneSpace
    alphaBoolBlend {blendOption=Linear}
    alphaScaleBiasClamp {bInterpResult=true, interpSpeedDecreasing=5, interpSpeedIncreasing=8}
    boneToModify {boneName=base_midbelt_l}
    rotation {pitch=20}
   }
   Save cached pose 'MainLopp'(Pose=Component To Local(ComponentPose=@n33)) {cacheName=MainLopp}
  ]}
 EventGraph {nodes [
   n1 = Event BlueprintProcessAnimInputEvent()
   > Set zip Off(zip Off:double=1) -> (Output_Get:double)
   > n2 = Branch(Condition=IsNamed(self=@n1.AnimInputEvent, InEventName=aie.Player.Zipline.Entered))
   > Set zip Off(zip Off=0)
   Branch(execute=@n2.else, Condition=IsNamed(self=@n1.AnimInputEvent, InEventName=aie.Player.Zipline.Exited))
   > Set zip Off(zip Off=1)
   n3 = Event BlueprintUpdateAnimation()
   > Set DeltaTimeLocal(DeltaTimeLocal:double=@n3.DeltaTimeX) -> (Output_Get:double)
   > n4 = Sequence!(execute) -> (then_0, then_1, then_2:Exec)
   > n5 = Gravity Override State Function DEPRECATED(CharacterPartAnimInstance=Self-Reference(), GameTimeInSeconds=GetGameTimeInSeconds(), GravityOverrideParams=Get Gravity Override Params())
   > Set GravityOverride(GravityOverride=@n5)
   n6 = NOT Boolean(A=Get bIsMoving2D())
   n7 = Get bIsCrouching()
   Set Crouching(execute=@n4.then_1, Crouching:double="To Float (Boolean)"(InBool=AND Boolean(A=@n7, B=@n6))) -> (Output_Get:double)
   n8 = Get bIsCrouching()
   n9 = Get bIsMoving2D()
   > Set CrouchMove(CrouchMove:double="To Float (Boolean)"(InBool=AND Boolean(A=@n8, B=@n9))) -> (Output_Get:double)
   n10 = Get bIsCrouching()
   > Set NotCrouching(NotCrouching:double="To Float (Boolean)"(InBool=NOT Boolean(A=@n10))) -> (Output_Get:double)
   > Set JUMP(JUMP:double="To Float (Boolean)"(InBool=Get bIsJumping())) -> (Output_Get:double)
   n11 = Get bIsJumping()
   > Set NotJUMP(NotJUMP:double="To Float (Boolean)"(InBool=NOT Boolean(A=@n11))) -> (Output_Get:double)
   n12 = Get bIsPlayingEmote()
   > Set NoEmote(NoEmote:double="To Float (Boolean)"(InBool=NOT Boolean(A=@n12))) -> (Output_Get:double)
   n13 = NOT Boolean(A=Get bIsInFrontEnd())
   n14 = OR Boolean(A=Get bIsPlayingEmote(), B=Get bIsOnGrindRail())
   > Set Emote(Emote:double="To Float (Boolean)"(InBool=AND Boolean(A=@n14, B=@n13))) -> (Output_Get:double)
   > Set Swing(Swing:double="To Float (Boolean)"(InBool=Get bIsPlayingMeleeAnim())) -> (Output_Get:double)
   > Set skydivefloating(skydivefloating:bool=OR Boolean(A=Get bIsSkydiveFloating(), B=Get bIsParachuteOpened())) -> (Output_Get:bool)
   n15 = NOT Boolean(A=Get bIsPlayingEmote())
   n16 = Get bIsInFrontEnd()
   > Set FE_NoEmote(FE_NoEmote:double="To Float (Boolean)"(InBool=AND Boolean(A=@n16, B=@n15))) -> (Output_Get:double)
   > Set DynOff(DynOff:double=0) -> (Output_Get:double)
   FInterpTo(Current=Get FE_NoEmote(), Target=0, DeltaTime=Get DeltaTimeLocal() -> (DeltaTimeLocal:double), InterpSpeed=1)
   n17 = Get bIsInFrontEnd()
   n18 = Get zip Off()
   "To Float (Boolean)"(InBool=OR Boolean(A="float > float"(A=@n18), B=NOT Boolean(A=@n17)))
  ]}
]
variables [
 BodyMesh {type=SkeletalMeshComponent}
 Crouching {type=number}
 CrouchMove {type=number}
 DBNO {type=number}
 DeltaTimeLocal {type=number}
 DynOff {type=number}
 Emote {type=number}
 emote_alpha {type=number}
 FE_NoEmote {type=number}
 Gravity Override Params {type=GravityOverrideParamsStruct}
 GravityOverride {type=Vector}
 isSkydiving {type=number}
 isTargeting {type=number}
 JUMP {type=number}
 NoEmote {type=number}
 NotCrouching {type=number}
 NotJUMP {type=number}
 notSkydiveing {type=number}
 skydivefloating {type=boolean}
 Swing {type=number}
 zip Off {type=number}
 Zipline {type=boolean}
]
