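# Physics asset for the hair on the F_MED_NanoChain female character's face accessory mesh, covering bangs and a ponytail.
# Large capsules on spine, neck and head form the body shapes, linked by free-rotation constraints.
# Six bang chains hang from the head through small base capsules, with mostly tight swing and zero twist limits on dynamic links.
# The ponytail branches into four strand chains, the central one forking left and right.
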PhysicsAsset /BRCosmetics/Characters/Player/Female/Medium/Bodies/F_MED_NanoChain/Meshes/Parts/F_MED_NanoChain_FaceAcc_Physics
bodies {
 base_bangsbackleft [{shapeType=Capsule, length=0.505, radius=0.505, shapeName=base_bangsbackleft_capsule, rotation=<rot 0 -90 0>}]
 base_bangsbackright [{shapeType=Capsule, length=0.505, radius=0.505, shapeName=base_bangsbackright_capsule, rotation=<rot 0 -90 0>}]
 base_bangsfrontWletft [{shapeType=Capsule, length=0.505, radius=0.505, shapeName=base_bangsfrontWletft_capsule, rotation=<rot 0 -90 0>}]
 base_bangsfrontleft [{shapeType=Capsule, length=0.505, radius=0.505, shapeName=base_bangsfrontleft_capsule, rotation=<rot 0 -90 0>}]
 base_bangsfrontright [{shapeType=Capsule, length=0.505, radius=0.505, shapeName=base_bangsfrontright_capsule, rotation=<rot 0 -90 0>}]
 base_bangsmidright [{shapeType=Capsule, length=0.505, radius=0.505, shapeName=base_bangsmidright_capsule, rotation=<rot 0 -90 0>}]
 base_centralstrand [{shapeType=Capsule, length=0.505, radius=0.505, shapeName=base_centralstrand_capsule}]
 base_exleftstrand [{shapeType=Capsule, length=0.505, radius=0.505, shapeName=base_exleftstrand_capsule}]
 base_lowerstrand [{shapeType=Capsule, length=0.505, radius=0.505, shapeName=base_lowerstrand_capsule}]
 base_ponytail [{shapeType=Capsule, length=0.505, radius=0.505, shapeName=base_ponytail_capsule}]
 base_rightstrand [{shapeType=Capsule, length=0.505, radius=0.505, shapeName=base_rightstrand_capsule}]
 dyn_bangsbackleft [{shapeType=Capsule, length=2, radius=1, shapeName=dyn_bangsbackleft_capsule, center=<0 0 2>}]
 dyn_bangsbackleft_02 [{shapeType=Capsule, length=2, radius=1, shapeName=dyn_bangsbackleft_02_capsule, center=<0 0 1.2076>}]
 dyn_bangsbackleft_03 [{shapeType=Capsule, length=2, radius=1, shapeName=dyn_bangsbackleft_03_capsule, center=<0 0.2528 2.3939>}]
 dyn_bangsbackright [{shapeType=Capsule, length=2, radius=1, shapeName=dyn_bangsbackright_capsule, center=<0 0 2>}]
 dyn_bangsbackright_02 [{shapeType=Capsule, length=2, radius=1, shapeName=dyn_bangsbackright_02_capsule, center=<0 0 2>}]
 dyn_bangsbackright_03 [{shapeType=Capsule, length=2, radius=1, shapeName=dyn_bangsbackright_03_capsule, center=<0 0 2>}]
 dyn_bangsfrontWletft [{shapeType=Capsule, length=1, radius=0.5, shapeName=dyn_bangsfrontWletft_capsule, center=<0.1896 0.2264 2>}]
 dyn_bangsfrontWletft_02 [{shapeType=Capsule, length=2, radius=1, shapeName=dyn_bangsfrontWletft_02_capsule}]
 dyn_bangsfrontWletft_03 [{shapeType=Capsule, length=1, radius=1, shapeName=dyn_bangsfrontWletft_03_capsule, center=<0.7471 -0.534 1.7998>, rotation=<rot -15 0 0.0002>}]
 dyn_bangsfrontleft [{shapeType=Capsule, length=1.2, radius=1, shapeName=dyn_bangsfrontleft_capsule}]
 dyn_bangsfrontleft_02 [{shapeType=Capsule, length=4, radius=1.2, shapeName=dyn_bangsfrontleft_02_capsule, center=<0 0.4824 1.7989>, rotation=<rot 0 0 5.0001>}]
 dyn_bangsfrontleft_03 [{shapeType=Capsule, length=2, radius=1, shapeName=dyn_bangsfrontleft_03_capsule, center=<0 0 2>}]
 dyn_bangsfrontleft_04 [{shapeType=Capsule, length=2, radius=1, shapeName=dyn_bangsfrontleft_04_capsule, center=<0 0.5529 2.4257>}]
 dyn_bangsfrontright [{shapeType=Capsule, length=2, radius=0.6, shapeName=dyn_bangsfrontright_capsule, center=<0 0.2363 2.3214>}]
 dyn_bangsfrontright_02 [{shapeType=Capsule, length=3, radius=0.5, shapeName=dyn_bangsfrontright_02_capsule, center=<0 0.2227 0.8316>, rotation=<rot 0 0 11.2501>}]
 dyn_bangsfrontright_03 [{shapeType=Capsule, length=1, radius=0.5, shapeName=dyn_bangsfrontright_03_capsule, center=<0 0 1.6045>}]
 dyn_bangsmidright [{shapeType=Capsule, length=2, radius=1, shapeName=dyn_bangsmidright_capsule, center=<0.1971 -0.3094 2>}]
 dyn_bangsmidright_02 [{shapeType=Capsule, length=7, radius=1, shapeName=dyn_bangsmidright_02_capsule, center=<0.1576 0 2>}]
 dyn_bangsmidright_03 [{shapeType=Capsule, length=3, radius=1, shapeName=dyn_bangsmidright_03_capsule, center=<0 -1.0678 2.9909>}]
 dyn_centralstrand [{shapeType=Capsule, length=2, radius=2, shapeName=dyn_centralstrand_capsule, center=<0 0 2>}]
 dyn_centralstrand_l [{shapeType=Capsule, length=2, radius=2, shapeName=dyn_centralstrand_l_capsule, center=<0 0 1>}]
 dyn_centralstrand_r [{shapeType=Capsule, length=2, radius=2, shapeName=dyn_centralstrand_r_capsule, center=<0 0 1>}]
 dyn_exleftstrand [{shapeType=Capsule, length=2, radius=2, shapeName=dyn_exleftstrand_capsule, center=<0 0 2>}]
 dyn_exleftstrand_02 [{shapeType=Capsule, length=2, radius=2, shapeName=dyn_exleftstrand_02_capsule, center=<0 0 1>}]
 dyn_lowerstrand [{shapeType=Capsule, length=2, radius=2, shapeName=dyn_lowerstrand_capsule, center=<0 0 2>}]
 dyn_lowerstrand_02 [{shapeType=Capsule, length=2, radius=2, shapeName=dyn_lowerstrand_02_capsule, center=<0 0 1>}]
 dyn_ponytail [{shapeType=Capsule, length=2, radius=2, shapeName=dyn_ponytail_capsule, center=<0 0 5>}]
 dyn_rightstrand [{shapeType=Capsule, length=2, radius=2, shapeName=dyn_rightstrand_capsule, center=<0 0 2>}]
 dyn_rightstrand_02 [{shapeType=Capsule, length=2, radius=2, shapeName=dyn_rightstrand_02_capsule, center=<0 0 1>}]
 head [{shapeType=Capsule, length=10.5949, radius=6.6, shapeName=head_capsule, center=<0 1.5684 0>, rotation=<rot 90 0 0>}]
 neck_01 [{shapeType=Capsule, length=2, radius=8.5, shapeName=neck_01_capsule, rotation=<rot 90 0 0>}]
 spine_05 [{shapeType=Capsule, length=8, radius=10, shapeName=spine_05_capsule, center=<2.038 0 0>}]
}
constraints [
 {bone1Name=base_bangsbackleft, bone2Name=head, swing1LimitDegrees=45, swing2LimitDegrees=45, twistLimitDegrees=45}
 {bone1Name=dyn_bangsbackleft, bone2Name=base_bangsbackleft, swing1LimitDegrees=25, swing2LimitDegrees=25, twistLimitDegrees=0}
 {bone1Name=dyn_bangsbackleft_02, bone2Name=dyn_bangsbackleft, swing1LimitDegrees=35, swing2LimitDegrees=35, twistLimitDegrees=0}
 {bone1Name=dyn_bangsbackleft_03, bone2Name=dyn_bangsbackleft_02, swing1LimitDegrees=35, swing2LimitDegrees=35, twistLimitDegrees=0}
 {bone1Name=base_bangsbackright, bone2Name=head, swing1LimitDegrees=45, swing2LimitDegrees=45, twistLimitDegrees=45}
 {bone1Name=dyn_bangsbackright, bone2Name=base_bangsbackright, swing1LimitDegrees=20, swing2LimitDegrees=20, twistLimitDegrees=0}
 {bone1Name=dyn_bangsbackright_02, bone2Name=dyn_bangsbackright, swing1LimitDegrees=35, swing2LimitDegrees=35, twistLimitDegrees=0}
 {bone1Name=dyn_bangsbackright_03, bone2Name=dyn_bangsbackright_02, swing1LimitDegrees=35, swing2LimitDegrees=35, twistLimitDegrees=0}
 {bone1Name=base_bangsfrontleft, bone2Name=head, swing1LimitDegrees=45, swing2LimitDegrees=45, twistLimitDegrees=45}
 {bone1Name=dyn_bangsfrontleft, bone2Name=base_bangsfrontleft, swing1LimitDegrees=25, swing2LimitDegrees=25, twistLimitDegrees=0}
 {bone1Name=dyn_bangsfrontleft_02, bone2Name=dyn_bangsfrontleft, swing1LimitDegrees=20, swing2LimitDegrees=5, twistLimitDegrees=0}
 {bone1Name=dyn_bangsfrontleft_03, bone2Name=dyn_bangsfrontleft_02, swing1LimitDegrees=35, swing2LimitDegrees=35, twistLimitDegrees=0}
 {bone1Name=dyn_bangsfrontleft_04, bone2Name=dyn_bangsfrontleft_03, swing1LimitDegrees=35, swing2LimitDegrees=35, twistLimitDegrees=0}
 {bone1Name=base_bangsfrontright, bone2Name=head, swing1LimitDegrees=45, swing2LimitDegrees=45, twistLimitDegrees=45}
 {bone1Name=dyn_bangsfrontright, bone2Name=base_bangsfrontright, swing1LimitDegrees=20, swing2LimitDegrees=5, twistLimitDegrees=0}
 {bone1Name=dyn_bangsfrontright_02, bone2Name=dyn_bangsfrontright, swing1LimitDegrees=35, swing2LimitDegrees=35, twistLimitDegrees=2}
 {bone1Name=dyn_bangsfrontright_03, bone2Name=dyn_bangsfrontright_02, swing1LimitDegrees=35, swing2LimitDegrees=35, twistLimitDegrees=0}
 {bone1Name=base_bangsmidright, bone2Name=head, swing1LimitDegrees=45, swing2LimitDegrees=45, twistLimitDegrees=45}
 {bone1Name=dyn_bangsmidright, bone2Name=base_bangsmidright, swing1LimitDegrees=20, swing2LimitDegrees=5, twistLimitDegrees=0}
 {bone1Name=dyn_bangsmidright_02, bone2Name=dyn_bangsmidright, swing1LimitDegrees=35, swing2LimitDegrees=35, twistLimitDegrees=0}
 {bone1Name=dyn_bangsmidright_03, bone2Name=dyn_bangsmidright_02, swing1LimitDegrees=35, swing2LimitDegrees=35, twistLimitDegrees=0}
 {bone1Name=base_bangsfrontWletft, bone2Name=head, swing1LimitDegrees=45, swing2LimitDegrees=45, twistLimitDegrees=45}
 {bone1Name=dyn_bangsfrontWletft, bone2Name=base_bangsfrontWletft, swing1LimitDegrees=5, swing2LimitDegrees=5, twistLimitDegrees=0}
 {bone1Name=dyn_bangsfrontWletft_02, bone2Name=dyn_bangsfrontWletft, swing1LimitDegrees=20, swing2LimitDegrees=20, twistLimitDegrees=0}
 {bone1Name=dyn_bangsfrontWletft_03, bone2Name=dyn_bangsfrontWletft_02, swing1LimitDegrees=20, swing2LimitDegrees=20, twistLimitDegrees=0}
 {
  bone1Name=head
  bone2Name=neck_01
  swing1LimitDegrees=45
  swing1Motion=Free
  swing2LimitDegrees=45
  swing2Motion=Free
  twistLimitDegrees=45
  twistMotion=Free
 }
 {
  bone1Name=neck_01
  bone2Name=spine_05
  swing1LimitDegrees=45
  swing1Motion=Free
  swing2LimitDegrees=45
  swing2Motion=Free
  twistLimitDegrees=45
  twistMotion=Free
 }
 {bone1Name=base_ponytail, bone2Name=head, swing1LimitDegrees=45, swing2LimitDegrees=45, twistLimitDegrees=45}
 {bone1Name=dyn_ponytail, bone2Name=base_ponytail, swing1LimitDegrees=20, swing2LimitDegrees=20, twistLimitDegrees=0}
 {bone1Name=base_centralstrand, bone2Name=dyn_ponytail, swing1LimitDegrees=45, swing2LimitDegrees=45, twistLimitDegrees=45}
 {bone1Name=dyn_centralstrand, bone2Name=base_centralstrand, swing1LimitDegrees=15, swing2LimitDegrees=15, twistLimitDegrees=2}
 {bone1Name=dyn_centralstrand_l, bone2Name=dyn_centralstrand, swing1LimitDegrees=15, swing2LimitDegrees=15, twistLimitDegrees=2}
 {bone1Name=dyn_centralstrand_r, bone2Name=dyn_centralstrand, swing1LimitDegrees=15, swing2LimitDegrees=15, twistLimitDegrees=2}
 {bone1Name=base_exleftstrand, bone2Name=dyn_ponytail, swing1LimitDegrees=45, swing2LimitDegrees=45, twistLimitDegrees=45}
 {bone1Name=dyn_exleftstrand, bone2Name=base_exleftstrand, swing1LimitDegrees=15, swing2LimitDegrees=15, twistLimitDegrees=2}
 {bone1Name=dyn_exleftstrand_02, bone2Name=dyn_exleftstrand, swing1LimitDegrees=15, swing2LimitDegrees=15, twistLimitDegrees=2}
 {bone1Name=base_lowerstrand, bone2Name=dyn_ponytail, swing1LimitDegrees=45, swing2LimitDegrees=45, twistLimitDegrees=45}
 {bone1Name=dyn_lowerstrand, bone2Name=base_lowerstrand, swing1LimitDegrees=15, swing2LimitDegrees=15, twistLimitDegrees=2}
 {bone1Name=dyn_lowerstrand_02, bone2Name=dyn_lowerstrand, swing1LimitDegrees=15, swing2LimitDegrees=15, twistLimitDegrees=2}
 {bone1Name=base_rightstrand, bone2Name=dyn_ponytail, swing1LimitDegrees=45, swing2LimitDegrees=45, twistLimitDegrees=45}
 {bone1Name=dyn_rightstrand, bone2Name=base_rightstrand, swing1LimitDegrees=15, swing2LimitDegrees=15, twistLimitDegrees=2}
 {bone1Name=dyn_rightstrand_02, bone2Name=dyn_rightstrand, swing1LimitDegrees=15, swing2LimitDegrees=15, twistLimitDegrees=2}
]
